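; PCG instant graph that shuffles actors by randomly redistributing their transforms among one another.
; A shared input subgraph supplies selected or referenced actors, with references resolved to single points and gathered with the selection.
; The Seed parameter mutates point seeds, and sorting by seed creates a random permutation.
; ActorReference values from the sorted copy are written onto the unsorted points, then each actor teleports to its new point.
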(graph Instant_Transform_ShuffleActors
  (params (Seed :type Int :default 67029422) (UseSelectedActors :type Bool :default true) (ActorReferences :type (Array (SoftObject /Script/Engine.Actor))))
  (bind Subgraph_0 (Subgraph :Graph /PCGPrimitives/Instants/Shared/Input_ActorRefsOrSelection.Input_ActorRefsOrSelection :ActorReferences ActorReferences :UseSelectedActors UseSelectedActors))
  (bind GetActorData_0 (Get%20Actor%20Data :ActorSelector (struct :ActorFilter FromInput :ActorReferenceSelector ActorReferences) :bMergeSinglePointData true :Mode GetSinglePoint))
  (bind Gather_0 (Gather))
  (bind MutateSeed_1 (Mutate%20Seed :Seed-param Seed))
  (bind SortAttributes_0 (Sort%20Attributes :bUseStableSort false :SortAttributes [(struct :InputSource $Seed)]))
  (bind CopyAttributes_0 (Copy%20Attributes :InputSource ActorReference :OutputTarget ActorReference))
  (bind TeleportActorsAndComponents_0 (Teleport%20Actors%20And%20Components :ObjectReferenceAttribute ActorReference))
  (connect SortAttributes_0 Out CopyAttributes_0 Source)
  (connect MutateSeed_1 Out CopyAttributes_0 Target)
  (connect GetActorData_0 Out Gather_0 In)
  (connect Subgraph_0 SelectedActors Gather_0 In)
  (connect Subgraph_0 ActorReferences GetActorData_0 In)
  (connect Gather_0 Out MutateSeed_1 In)
  (connect MutateSeed_1 Out SortAttributes_0 In)
  (connect CopyAttributes_0 Out TeleportActorsAndComponents_0 In))
