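; PCG instant tool graph that randomly jitters spline point positions on selected or referenced actors and writes the result back to their spline components.
; It extracts spline data from those actors, printing an error when none is found.
; Attribute noise offsets X and Y, plus Z when JitterIn3D is set, within plus or minus JitterAmount, using chained seed generation.
; Partitioned spline component references supply the targets for Apply Spline To Component.
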(graph Instant_Transform_Spline_Jitter
  (params (Seed :type Int :default 42721838) (UseSelectedActors :type Bool :default true) (ActorReferences :type (Array (SoftObject /Script/Engine.Actor))) (JitterAmount :type Float) (JitterIn3D :type Bool))
  (bind Subgraph_0 (Subgraph :Graph /PCGPrimitives/Instants/Shared/Input_ActorRefsOrSelection.Input_ActorRefsOrSelection :ActorReferences ActorReferences :UseSelectedActors UseSelectedActors))
  (bind GetActorData_0 (Get%20Actor%20Data :ActorSelector (struct :ActorFilter FromInput :ActorReferenceSelector ActorReferences) :bMergeSinglePointData true :Mode GetSinglePoint))
  (bind Gather_0 (Gather))
  (bind AttributeMathsOp_2 (Attribute%20Maths%20Op :Operation Negate :In JitterAmount))
  (bind GetActorData_1 (Get%20Actor%20Data :ActorSelector (struct :ActorFilter FromInput :ActorReferenceSelector ActorReference) :bIgnorePCGGeneratedComponents false))
  (bind FilterDataByType_0 (Filter%20Data%20By%20Type :TargetType (struct :Ids [(struct :Struct /Script/PCG.PCGDataTypeInfoSpline)])))
  (bind DataNum_0 (Data%20Count))
  (bind Branch_0 (Branch))
  (bind Branch_1 (Branch :bOutputToB-param JitterIn3D))
  (bind GenerateSeed_0 (Generate%20Seed :GenerationSource HashEachSourceAttribute :OutputTarget Seed :SeedSource Seed :In Seed))
  (bind AttributeNoise_0 (Attribute%20Noise :InputSource $Position.x :Mode Add :NoiseMax-param JitterAmount))
  (bind GenerateSeed_1 (Generate%20Seed :bResetSeedPerInput false :GenerationSource HashEachSourceAttribute :OutputTarget Seed :SeedSource Seed :In Seed))
  (bind AttributeNoise_1 (Attribute%20Noise :InputSource $Position.y :Mode Add :NoiseMax-param JitterAmount))
  (bind AttributeNoise_2 (Attribute%20Noise :InputSource $Position.x :Mode Add :NoiseMax-param JitterAmount))
  (bind AttributeNoise_3 (Attribute%20Noise :InputSource $Position.y :Mode Add :NoiseMax-param JitterAmount))
  (bind GenerateSeed_2 (Generate%20Seed :bResetSeedPerInput false :GenerationSource HashEachSourceAttribute :OutputTarget Seed :SeedSource Seed :In Seed))
  (bind AttributeNoise_4 (Attribute%20Noise :InputSource $Position.z :Mode Add :NoiseMax-param JitterAmount))
  (bind GetActorData_2 (Get%20Actor%20Data :ActorSelector (struct :ActorFilter FromInput :ActorReferenceSelector ActorReference) :bIgnorePCGGeneratedComponents false :ComponentSelector (struct :ComponentSelection ByClass :ComponentSelectionClass /Script/Engine.SplineComponent) :Mode GetComponentsReference))
  (bind MergeAttributes_0 (Merge%20Attributes))
  (bind AttributePartition_0 (Attribute%20Partition :PartitionAttributeSelectors [$Index]))
  (bind DeleteAttributes_0 (Delete%20Attributes :Operation KeepSelectedAttributes :SelectedAttributes "ComponentReference"))
  (bind ApplySplineToComponent_0 (Apply%20Spline%20To%20Component))
  (bind Print%20String_0 (Print%20String :bPrefixWithNode false :bPrintPerComponent false :PrintString "No spline actor selected" :Verbosity Error))
  (connect DeleteAttributes_0 Out ApplySplineToComponent_0 Component)
  (connect AttributeNoise_1 Out ApplySplineToComponent_0 Spline)
  (connect AttributeNoise_4 Out ApplySplineToComponent_0 Spline)
  (connect Branch_1 Output%20A AttributeNoise_0 In)
  (connect AttributeMathsOp_2 Out AttributeNoise_0 NoiseMin)
  (connect GenerateSeed_0 Out AttributeNoise_0 Seed)
  (connect AttributeNoise_0 Out AttributeNoise_1 In)
  (connect AttributeMathsOp_2 Out AttributeNoise_1 NoiseMin)
  (connect GenerateSeed_1 Out AttributeNoise_1 Seed)
  (connect Branch_1 Output%20B AttributeNoise_2 In)
  (connect AttributeMathsOp_2 Out AttributeNoise_2 NoiseMin)
  (connect GenerateSeed_0 Out AttributeNoise_2 Seed)
  (connect AttributeNoise_2 Out AttributeNoise_3 In)
  (connect AttributeMathsOp_2 Out AttributeNoise_3 NoiseMin)
  (connect GenerateSeed_1 Out AttributeNoise_3 Seed)
  (connect AttributeNoise_3 Out AttributeNoise_4 In)
  (connect AttributeMathsOp_2 Out AttributeNoise_4 NoiseMin)
  (connect GenerateSeed_2 Out AttributeNoise_4 Seed)
  (connect MergeAttributes_0 Out AttributePartition_0 In)
  (connect DataNum_0 Out Branch_0 bOutputToB)
  (connect FilterDataByType_0 InsideFilter Branch_0 In)
  (connect Branch_0 Output%20B Branch_1 In)
  (connect FilterDataByType_0 InsideFilter DataNum_0 In)
  (connect AttributePartition_0 Out DeleteAttributes_0 In)
  (connect GetActorData_1 Out FilterDataByType_0 In)
  (connect GetActorData_0 Out Gather_0 In)
  (connect Subgraph_0 SelectedActors Gather_0 In)
  (connect Subgraph_0 ActorReferences GetActorData_0 In)
  (connect Gather_0 Out GetActorData_1 In)
  (connect Gather_0 Out GetActorData_2 In)
  (connect GetActorData_2 Out MergeAttributes_0 In)
  (connect Branch_0 Output%20A Print%20String_0 In))
