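; PCG instant graph that projects splines of selected or referenced actors onto world surfaces or landscape only, then writes them back to their spline components.
; A per-spline loop projects control points onto landscape data or a world ray hit query whose origin derives from combined point bounds and MaxRayLength.
; Tags preserve open/closed state for rebuilding with custom tangents, and DataIndex restores component order.
; An error prints when no spline is found.
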(graph Instant_Transform_Spline_ProjectToSurface
  (params (UseSelectedActors :type Bool :default true) (ActorReferences :type (Array (SoftObject /Script/Engine.Actor))) (MaxRayLength :type Float :default 1e+04) (ProjectOnLandscapeOnly :type Bool))
  (bind Subgraph_0 (Subgraph :Graph /PCGPrimitives/Instants/Shared/Input_ActorRefsOrSelection.Input_ActorRefsOrSelection :ActorReferences ActorReferences :UseSelectedActors UseSelectedActors))
  (bind GetActorData_0 (Get%20Actor%20Data :ActorSelector (struct :ActorFilter FromInput :ActorReferenceSelector ActorReferences) :bMergeSinglePointData true :Mode GetSinglePoint))
  (bind Gather_0 (Gather))
  (bind GetActorData_1 (Get%20Actor%20Data :ActorSelector (struct :ActorFilter FromInput :ActorReferenceSelector ActorReference) :bIgnorePCGGeneratedComponents false))
  (bind FilterDataByType_0 (Filter%20Data%20By%20Type :TargetType (struct :Ids [(struct :Struct /Script/PCG.PCGDataTypeInfoSpline)])))
  (bind GetActorData_2 (Get%20Actor%20Data :ActorSelector (struct :ActorFilter FromInput :ActorReferenceSelector ActorReference) :bIgnorePCGGeneratedComponents false :ComponentSelector (struct :ComponentSelection ByClass :ComponentSelectionClass /Script/Engine.SplineComponent) :Mode GetComponentsReference))
  (bind MergeAttributes_0 (Merge%20Attributes))
  (bind AttributePartition_0 (Attribute%20Partition :PartitionAttributeSelectors [$Index]))
  (bind DeleteAttributes_0 (Delete%20Attributes :Operation KeepSelectedAttributes :SelectedAttributes "ComponentReference"))
  (bind DataNum_0 (Data%20Count))
  (bind Branch_0 (Branch))
  (bind WriteDataIndex_0 (Write%20Data%20Index :IndexTag "DataIndex"))
  (bind FilterDataByAttribute_0 (Filter%20Data%20By%20Attribute :FilterMode FilterByValue :FilterOperator Equal :TargetAttribute @Data.$IsClosed :Threshold (struct :AttributeTypes (struct :Type Boolean))))
  (bind AddTags_0 (Add%20Tags :TagsToAdd "Spline"))
  (bind AddTags_1 (Add%20Tags :TagsToAdd "ClosedSpline"))
  (bind Loop_0 (Loop :Graph ProjectEachInputSpline :MaxRayLength MaxRayLength :ProjectOnLandscapeOnly ProjectOnLandscapeOnly))
  (bind FilterDataByTag_0 (Filter%20Data%20By%20Tag :SelectedTags "Spline"))
  (bind CreateSpline_1 (Create%20Spline :bApplyCustomTangents true :bUseInterpTypeAttribute true))
  (bind CreateSpline_3 (Create%20Spline :bApplyCustomTangents true :bClosedLoop true :bUseInterpTypeAttribute true))
  (bind Gather_1 (Gather))
  (bind SortTags_0 (Sort%20Data%20By%20Tag%20Value :Tag DataIndex))
  (bind ApplySplineToComponent_0 (Apply%20Spline%20To%20Component))
  (bind Print%20String_0 (Print%20String :bPrefixWithNode false :bPrintPerComponent false :PrintString "No spline actor selected" :Verbosity Error))
  (connect FilterDataByAttribute_0 InsideFilter AddTags_0 In)
  (connect FilterDataByAttribute_0 OutsideFilter AddTags_1 In)
  (connect DeleteAttributes_0 Out ApplySplineToComponent_0 Component)
  (connect SortTags_0 Out ApplySplineToComponent_0 Spline)
  (connect MergeAttributes_0 Out AttributePartition_0 In)
  (connect DataNum_0 Out Branch_0 bOutputToB)
  (connect FilterDataByType_0 InsideFilter Branch_0 In)
  (connect FilterDataByTag_0 InsideFilter CreateSpline_1 In)
  (connect FilterDataByTag_0 OutsideFilter CreateSpline_3 In)
  (connect FilterDataByType_0 InsideFilter DataNum_0 In)
  (connect AttributePartition_0 Out DeleteAttributes_0 In)
  (connect WriteDataIndex_0 Out FilterDataByAttribute_0 In)
  (connect Loop_0 Out FilterDataByTag_0 In)
  (connect GetActorData_1 Out FilterDataByType_0 In)
  (connect GetActorData_0 Out Gather_0 In)
  (connect Subgraph_0 SelectedActors Gather_0 In)
  (connect CreateSpline_1 Out Gather_1 In)
  (connect CreateSpline_3 Out Gather_1 In)
  (connect Subgraph_0 ActorReferences GetActorData_0 In)
  (connect Gather_0 Out GetActorData_1 In)
  (connect Gather_0 Out GetActorData_2 In)
  (connect AddTags_0 Out Loop_0 In)
  (connect AddTags_1 Out Loop_0 In)
  (connect GetActorData_2 Out MergeAttributes_0 In)
  (connect Branch_0 Output%20A Print%20String_0 In)
  (connect Gather_1 Out SortTags_0 In)
  (connect Branch_0 Output%20B WriteDataIndex_0 In))

(subgraph ProjectEachInputSpline
  (params (MaxRayLength :type Double :default 1e+04) (ProjectOnLandscapeOnly :type Bool))
  (bind GetSplineControlPoints_1 (Get%20Spline%20Control%20Points))
  (bind Branch_0 (Branch :bOutputToB-param ProjectOnLandscapeOnly))
  (bind MergePoints_1 (Merge%20Points :bMergeMetadata false :MergeTagsMode None))
  (bind CombinePoints_1 (Combine%20Points :bUseFirstPointTransform false))
  (bind ResetPointCenter_1 (Reset%20Point%20Center :PointCenterLocation (0.5 0.5 1.0)))
  (bind ExtractAttributeAtIndex_1 (Extract%20Attribute%20at%20Index :InputSource $Position :OutputAttributeName RayOrigin))
  (bind AttributeMathsOp_1 (Attribute%20Maths%20Op :InB MaxRayLength))
  (bind AttributeMathsOp_2 (Attribute%20Maths%20Op :Operation Multiply :InA MaxRayLength))
  (bind WorldRayHitQuery_1 (World%20Ray%20Hit%20Query :QueryParams (struct :bOverrideDefaultParams true)))
  (bind GetLandscapeData_1 (Get%20Landscape%20Data))
  (bind Gather_0 (Gather))
  (bind Projection_1 (Projection :bKeepZeroDensityPoints true))
  (connect ExtractAttributeAtIndex_1 Out AttributeMathsOp_1 InA)
  (connect GetSplineControlPoints_1 Out Branch_0 In)
  (connect MergePoints_1 Out CombinePoints_1 In)
  (connect ResetPointCenter_1 Out ExtractAttributeAtIndex_1 In)
  (connect GetLandscapeData_1 Out Gather_0 In)
  (connect WorldRayHitQuery_1 Out Gather_0 In)
  (connect Branch_0 Output%20B GetLandscapeData_1 Execution%20Dependency)
  (connect DefaultInputNode In GetSplineControlPoints_1 In)
  (connect Branch_0 Output%20A MergePoints_1 In)
  (connect Branch_0 Output%20A Projection_1 In)
  (connect Branch_0 Output%20B Projection_1 In)
  (connect Gather_0 Out Projection_1 Projection%20Target)
  (connect CombinePoints_1 Out ResetPointCenter_1 In)
  (connect Branch_0 Output%20A WorldRayHitQuery_1 Execution%20Dependency)
  (connect AttributeMathsOp_2 Out WorldRayHitQuery_1 RayLength)
  (connect AttributeMathsOp_1 Out WorldRayHitQuery_1 RayOrigin)
  (output Out Projection_1))
